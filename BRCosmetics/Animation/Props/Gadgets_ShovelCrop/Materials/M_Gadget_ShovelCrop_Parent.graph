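# Parent material for the ShovelCrop gadget prop, layering ear flap motion and glow onto the 2023 anime character shader.
# Left and right sphere masks on pre-skinned position drive world position offset: gust directions scaled by a sine wave along bounds-normalized local Y, with stiffness falloff.
# Separate sphere masks blend emissive into an electricity effect from panning texture and animated gradient noise in stabilized screen-space UVs, plus ear-mask texture glow.
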Material /BRCosmetics/Animation/Props/Gadgets_ShovelCrop/Materials/M_Gadget_ShovelCrop_Parent
graph {
 properties {
  bAutomaticallySetUsageInEditor=false
  bUseMaterialAttributes=true
  bUsedWithClothing=true
  bUsedWithMeshDeformer=true
  bUsedWithMorphTargets=true
  bUsedWithNiagaraMeshParticles=true
  bUsedWithSkeletalMesh=true
  bUsedWithStaticLighting=true
  pixelDepthOffsetMode=PDOM_Legacy
  previewMesh=/BRCosmetics/Animation/Props/Gadgets_ShovelCrop/Meshes/ShovelCrop_Gadget
  shadingModel=MSM_Cloth
  usedShadingModels=Cloth
 }
 nodes [
  n1 = VertexInterpolator(VS=LocalPosition() {localOrigin=InstancePreSkinning}.XYZ)
  n2 = ScalarParameter() {group=Ears Flap WPO, parameterName=Ear Mask Radius}
  n3 = ScalarParameter() {group=Ears Flap WPO, parameterName=Ear Mask Hardness}
  Add(
   A=SphereMask(
    A=@n1.PS
    B=VectorParameter() {group=Ears Flap WPO, parameterName=Right Ear Mask Pos, defaultValue=<rgba 1 0 0 0>}.RGB
    Radius=@n2
    Hardness=@n3
   )
   B=SphereMask(
    A=@n1.PS
    B=VectorParameter() {group=Ears Flap WPO, parameterName=Left Ear Mask Pos, defaultValue=<rgba 1 0 0 0>}.RGB
    Radius=@n2
    Hardness=@n3
   )
  )
  MaterialFunctionCall(
   Brightness=ScalarParameter() {defaultValue=1, group=02_Anime Parameters, parameterName=Emissive_Rim_Brightness, sliderMax=10}
   Exponent=ScalarParameter() {defaultValue=1, group=02_Anime Parameters, parameterName=Emissive_Rim_Exp, sliderMax=10}
  ) {materialFunction=&MF_Fresnel_FX}
  n4 = Multiply(
   A=MaterialFunctionCall!(ObjectPivotLocation, ObjectPositionOffset=Constant3Vector()) {materialFunction=/Game/Characters/Player/Common/MaterialFunctions/Superlevel/ImprovedStabilizedScreenSpaceUVs}.StabilizedUVs
   B=ScalarParameter() {defaultValue=4, parameterName=CloudsScale}
  )
  n5 = Panner(Coordinate=@n4) {speedX=0.5}
  n6 = AppendVector(A=@n4, B=Multiply(A=Time(), B=ScalarParameter() {defaultValue=0.3, parameterName=CloudsAnimSpeed}))
  n7 = Power(
   Base=ConstantBiasScale(None=Noise!(World Position=@n6, FilterWidth) {bTiling=true, noiseFunction=NOISEFUNCTION_GradientTex3D, repeatSize=4})
   Exp=ScalarParameter() {defaultValue=2, parameterName=CloudsExp}
  )
  n8 = Multiply(
   A=Multiply(
    A=TextureSampleParameter2D(UVs=@n5) {parameterName=Inner Lightening Noise, texture=/Game/Global/Textures/Default/Masks_and_Gradients/chunkma_2}.R
    B=Saturate(None=@n7)
   )
   B=ScalarParameter() {defaultValue=3, parameterName=CloudsBoost}
  ) {constA=3}
  NamedRerouteDeclaration(None=LinearInterpolate(A=ScalarParameter() {defaultValue=0.03, parameterName=MinBrightness}, Alpha=Add(B=@n8)) {constA=0.03}) {name=InnerElectricityEffect, nodeColor=<rgba 1 0 0.1172 1>}
  n9 = MaterialFunctionCall!(CustomUV) {materialFunction=/Game/Global/Materials/Character_Parent_2023/Anime_Parent_2023/MaterialFunctions/MFC_AnimeShader_2023_Pre}
  n10 = MaterialFunctionCall!(CustomLightVector, OverrideNormal_WS, STT=@n9.STT, Normal_Map=@n9.STT, DFL=@n9.STT, FX=@n9.STT) {materialFunction=/Game/Global/Materials/Character_Parent_2023/Anime_Parent_2023/MaterialFunctions/MFC_AnimeShader_2023_Calculations}
  n11 = VectorParameter() {group=01_Anime Maps, parameterName=Emissive_Color, sortPriority=20, defaultValue=<rgba 1 0.9653 0.4688 1>}
  n12 = Multiply(A=@n11.RGB, B=ScalarParameter() {defaultValue=0.8, group=02_Anime Parameters, parameterName=Emissive_Intensity, sliderMax=10})
  n13 = VertexInterpolator(VS=LocalPosition() {localOrigin=InstancePreSkinning}.XYZ)
  n14 = VectorParameter() {group=Ears Flap WPO, parameterName=Right Emissive Mask Pos, defaultValue=<rgba 1 0 0 0>}
  n15 = ScalarParameter() {group=Ears Flap WPO, parameterName=Emissive Mask Radius}
  n16 = ScalarParameter() {group=Ears Flap WPO, parameterName=Emissive Mask Hardness}
  n17 = VectorParameter() {group=Ears Flap WPO, parameterName=Left Emissive Mask Pos, defaultValue=<rgba 1 0 0 0>}
  n18 = LinearInterpolate(
   A=@n9.Emissive
   B=Multiply(A=@n12, B=NamedRerouteUsage())
   Alpha=Add(A=SphereMask(A=@n13.PS, B=@n14.RGB, Radius=@n15, Hardness=@n16), B=SphereMask(A=@n13.PS, B=@n17.RGB, Radius=@n15, Hardness=@n16))
  )
  n19 = TextureSampleParameter2D(UVs=TextureCoordinate()) {
   parameterName=Ears Mask
   samplerType=SAMPLERTYPE_Masks
   texture=/BRCosmetics/Animation/Props/Gadgets_ShovelCrop/Textures/T_ShovelCrop_Emote_Mask
  }
  n20 = Multiply(A=@n12, B=Power(Base=@n19.R, Exp=ScalarParameter() {defaultValue=1, group=02_Anime Parameters, parameterName=Mask exp, sliderMax=10}))
  n21 = LocalPosition() {localOrigin=InstancePreSkinning}
  n22 = ScalarParameter() {group=Ears Flap WPO, parameterName=Ear Mask Radius}
  n23 = ScalarParameter() {group=Ears Flap WPO, parameterName=Ear Mask Hardness}
  n24 = LinearInterpolate(
   B=VectorParameter() {group=Ears Flap WPO, parameterName=Right Ear GustDir, defaultValue=<rgba 1 0 0 0>}.RGB
   Alpha=SphereMask(
    A=@n21.XYZ
    B=VectorParameter() {group=Ears Flap WPO, parameterName=Right Ear Mask Pos, defaultValue=<rgba 1 0 0 0>}.RGB
    Radius=@n22
    Hardness=@n23
   )
  )
  n25 = LinearInterpolate(
   B=VectorParameter() {group=Ears Flap WPO, parameterName=Left Ear GustDir, defaultValue=<rgba 1 0 0 0>}.RGB
   Alpha=SphereMask(
    A=@n21.XYZ
    B=VectorParameter() {group=Ears Flap WPO, parameterName=Left Ear Mask Pos, defaultValue=<rgba 1 0 0 0>}.RGB
    Radius=@n22
    Hardness=@n23
   )
  )
  n26 = ComponentMask(None=LocalPosition() {localOrigin=InstancePreSkinning}.XYZ) {g=true}
  n27 = PreSkinnedLocalBounds()
  n28 = ComponentMask(None=@n27.Min) {g=true}
  n29 = ComponentMask(None=@n27.Max) {g=true}
  n30 = Saturate(None=Divide(A=Subtract(A=@n26, B=@n28), B=Subtract(A=@n29, B=@n28)))
  n31 = ScalarParameter() {defaultValue=2, group=Ears Flap WPO, parameterName=WaveFreq}
  n32 = Time()
  n33 = ScalarParameter() {defaultValue=8, group=Ears Flap WPO, parameterName=WaveSpeed}
  n34 = Sine(None=Subtract(A=Multiply(A=@n30, B=@n31), B=Multiply(A=@n32, B=@n33)))
  n35 = Multiply(
   A=Multiply(A=Add(B=@n34) {constA=1}, B=ScalarParameter() {defaultValue=0.5, group=Ears Flap WPO, parameterName=FlapAmount})
   B=Power(Base=@n30, Exp=ScalarParameter() {defaultValue=2, group=Ears Flap WPO, parameterName=Stiffness})
  )
  n36 = ScalarParameter() {group=Ears Flap WPO, parameterName=Intensity}
  SetMaterialAttributes!(
   MaterialAttributes=MaterialFunctionCall!(
    Shading_Gradient=@n10.Shading_Gradient
    Threshold_Gradient=@n10.Shading_Gradient
    Specular_Gradient=@n10.Shading_Gradient
    Sheen_Mask=@n10.Shading_Gradient
    Thickness=@n10.Shading_Gradient
    DFL_Mask=@n10.Shading_Gradient
    FX=@n10.Shading_Gradient
    Metallic_Gradient=@n10.Shading_Gradient
    RimLight_Gradient=@n10.Shading_Gradient
    NormalBuffer=@n10.Shading_Gradient
    --------
    CINE_LightSources=@n10.Shading_Gradient
    CINE_LightColor=@n10.Shading_Gradient
    ---------
    Color_Lit=@n9.Color_Lit
    Color_Shaded=@n9.Color_Lit
    SRM=@n9.Color_Lit
    Emissive=Add(A=@n18, B=@n20)
    DFL_Color=@n9.Color_Lit
    Specular_Color=@n9.Color_Lit
    RimLight_Color=@n9.Color_Lit
   ) {materialFunction=/Game/Global/Materials/Character_Parent_2023/Anime_Parent_2023/MaterialFunctions/MFC_AnimeShader_2023_Post}.Result
   World Position Offset=Multiply(A=Add(A=@n24, B=@n25), B=Multiply(A=@n35, B=@n36))
  ) {attributeSetTypes [#1]}
 ]
}
