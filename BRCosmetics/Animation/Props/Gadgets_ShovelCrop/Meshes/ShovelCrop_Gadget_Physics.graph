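# Physics asset for the ShovelCrop gadget prop mesh, covering its head attach bone and left and right ears with capsule bodies.
# Each ear base joint carries two capsules, while each dynamic ear joint has a single thinner capsule, mirrored between sides.
# Two constraints tie each dynamic ear joint to its base joint with matching swing and tighter twist limits, allowing constrained secondary ear motion.
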PhysicsAsset /BRCosmetics/Animation/Props/Gadgets_ShovelCrop/Meshes/ShovelCrop_Gadget_Physics
bodies {
 head_attach [{shapeType=Capsule, length=0.505, radius=5.505, rotation=<rot 0 -90 0>}]
 l_ear_dyn_jnt [{shapeType=Capsule, length=4.238, radius=1.1608, center=<3.5146 0.4838 0.7432>, rotation=<rot 0 89.4529 -74.2618>}]
 l_ear_jnt [
  {shapeType=Capsule, length=3.5, radius=2, shapeName=l_ear_jnt_capsule, center=<3.7749 0 -3.1849>, rotation=<rot 0 -0.0001 89.9999>}
  {shapeType=Capsule, length=3, radius=1.5, shapeName=l_ear_jnt_capsule, center=<1.9331 3.8645 0>}
 ]
 r_ear_dyn_jnt [{shapeType=Capsule, length=4.238, radius=1.1608, center=<-3.5146 -0.4838 -0.7432>, rotation=<rot 0 89.4529 105.7382>}]
 r_ear_jnt [
  {shapeType=Capsule, length=3.5, radius=2, shapeName=r_ear_jnt_capsule, center=<-3.7749 0 3.1849>, rotation=<rot 0 -0.0001 -90.0001>}
  {shapeType=Capsule, length=3, radius=1.5, shapeName=r_ear_jnt_capsule, center=<-1.9331 -3.8645 0>, rotation=<rot 0 0 -180>}
 ]
}
constraints [
 {bone1Name=r_ear_dyn_jnt, bone2Name=r_ear_jnt, swing1LimitDegrees=30, swing2LimitDegrees=30, twistLimitDegrees=8}
 {bone1Name=l_ear_dyn_jnt, bone2Name=l_ear_jnt, swing1LimitDegrees=30, swing2LimitDegrees=30, twistLimitDegrees=8}
]
